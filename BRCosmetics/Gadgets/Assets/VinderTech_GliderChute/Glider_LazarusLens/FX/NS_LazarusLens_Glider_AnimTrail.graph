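# Niagara system that draws animated ribbon trails on the Lazarus Lens glider.
# Four left/right, outer/inner emitters and a Side_Tips emitter spawn short-lived ribbons between bone or socket pairs on User.SkeletalMesh.
# A clamped, system-level interpolated float fades particle alpha and the dark and teal secondary color attributes, with several ribbon renderers layered per emitter.
# A front-end modifier adds spawn velocity, which is zero during gameplay.
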NiagaraSystem /BRCosmetics/Gadgets/Assets/VinderTech_GliderChute/Glider_LazarusLens/FX/NS_LazarusLens_Glider_AnimTrail
dependencies {
 usedDataInterfaces []
 usedDynamicInputs [
  &Divide_Float
  &Boolean_AndOperation
  &SetBoolByFloatComparison
 ]
 usedModules [
  &SystemState
  &self:SystemSpawnScript.SystemScriptSource.SystemScriptGraph.NiagaraNodeAssignment_0.SetVariables_#1
  &InterpolateOverTime
  &Timeline
  &self:Set_Trail_Length
  &self:Set_Trail_Opacity
  &self:SystemSpawnScript.SystemScriptSource.SystemScriptGraph.NiagaraNodeAssignment_1.SetVariables_#2
 ]
}
emitters {
 FN_AnimTrail_Left {
  emitterSpawnScript [&FN_RibbonIDOffsetUtility()]
  emitterUpdateScript [
   &EmitterState()
   &FN_SpawnAnimTrail(Spawn Mode:ENiagaraSkeletalSpawning=::NewEnumerator0)
  ]
  particleSpawnScript [
   &AddVelocity(Velocity=&Lerp_Vector(A=<0 0 0>, B=<300 -800 0>, Alpha=&FN_FrontEndFloatModifier(Gameplay=0, Front End=1))) {module=AddVelocity001}
   &FN_InitializeAnimTrailBoneIndecies(First Bone or Socket Index=4, Second Bone or Socket Index=0)
   &FN_InitializeAnimTrail(Color=<rgba 0.9047 0.8148 0.7835 0.35>, Lifetime=0.15)
   /SetVariables(Particles.RibbonTwist2:NiagaraFloat=&AngleConversion(Angle:NiagaraFloat=0, Angle Input:ENiagara_AngleInput=::NewEnumerator0, Angle Output:ENiagara_AngleInput=::NewEnumerator2)) {isSetParameters=true, module=SetVariables_#3}
   /SetVariables(Particles.RibbonTwist:NiagaraFloat=&AngleConversion(Angle:NiagaraFloat=180, Angle Input:ENiagara_AngleInput=::NewEnumerator0, Angle Output:ENiagara_AngleInput=::NewEnumerator2)) {isSetParameters=true, module=SetVariables_#4}
  ]
  particleUpdateScript [
   &SolveForcesAndVelocity(Clamp Velocity=false, Limit Acceleration=false)
   &FN_UpdateAnimTrail(Locked Particle RibbonWidth=1, SkeletalMesh=$User.SkeletalMesh) {module=FN_UpdateAnimTrail001}
   &ScaleColor(
    Scale Alpha=&ClampFloat(Float=$System.InterpolateOverTime.InterpolateOverTime_Float.Moving Average)
    ScaleA=true
    ScaleRGB=true
   )
   /SetVariables(Particles.secondRibbonCOlor:LinearColor=&MultiplyLinearColor(
     Linear Color=<rgba 0.01 0 0 0.35>
     Scale Factor=&MakeFloat4FromVectorAndFloat(
      "Vector (RGB)"=<1 1 1>
      "Float (Alpha)"=&Multiply_Float(A=2, B=&ClampFloat(Float=$System.InterpolateOverTime.InterpolateOverTime_Float.Moving Average))
     )
    )) {isSetParameters=true, module=SetVariables_#5}
   /SetVariables(Particles.ThridRibbonCOlor:LinearColor=&MultiplyLinearColor(
     Linear Color=<rgba 0 0.5776 0.5457 0.5>
     Scale Factor=&MakeFloat4FromVectorAndFloat(
      "Vector (RGB)"=<1 1 1>
      "Float (Alpha)"=&Multiply_Float(A=2, B=&ClampFloat(Float=$System.InterpolateOverTime.InterpolateOverTime_Float.Moving Average))
     )
    )) {isSetParameters=true, module=SetVariables_#6}
  ]
  rendererClasses [&NiagaraRibbonRendererProperties]
  renderers [
   {rendererClass=&NiagaraRibbonRendererProperties}
   {rendererClass=&NiagaraRibbonRendererProperties, rendererIndex=1}
   {rendererClass=&NiagaraRibbonRendererProperties, rendererIndex=2}
   {rendererClass=&NiagaraRibbonRendererProperties, rendererIndex=3}
   {rendererClass=&NiagaraRibbonRendererProperties, rendererIndex=4}
  ]
 }
 FN_AnimTrail_Left_Inner {
  emitterSpawnScript [&FN_RibbonIDOffsetUtility()]
  emitterUpdateScript [
   &EmitterState()
   &FN_SpawnAnimTrail(Spawn Mode:ENiagaraSkeletalSpawning=::NewEnumerator0)
  ]
  particleSpawnScript [
   &AddVelocity(Velocity=&Lerp_Vector(A=<0 0 0>, B=<300 -800 0>, Alpha=&FN_FrontEndFloatModifier(Gameplay=0, Front End=1))) {module=AddVelocity001}
   &FN_InitializeAnimTrailBoneIndecies(First Bone or Socket Index=1, Second Bone or Socket Index=4) {module=FN_InitializeAnimTrailBoneIndecies001}
   &FN_InitializeAnimTrail(Color=<rgba 0.9047 0.8148 0.7835 0.35>, Lifetime=0.15)
   /SetVariables(Particles.RibbonTwist2:NiagaraFloat=&AngleConversion(Angle:NiagaraFloat=0, Angle Input:ENiagara_AngleInput=::NewEnumerator0, Angle Output:ENiagara_AngleInput=::NewEnumerator2)) {isSetParameters=true, module=SetVariables_#3}
   /SetVariables(Particles.RibbonTwist:NiagaraFloat=&AngleConversion(Angle:NiagaraFloat=180, Angle Input:ENiagara_AngleInput=::NewEnumerator0, Angle Output:ENiagara_AngleInput=::NewEnumerator2)) {isSetParameters=true}
  ]
  particleUpdateScript [
   &SolveForcesAndVelocity(Clamp Velocity=false, Limit Acceleration=false)
   &FN_UpdateAnimTrail(Locked Particle RibbonWidth=1, SkeletalMesh=$User.SkeletalMesh) {module=FN_UpdateAnimTrail001}
   &ScaleColor(
    Scale Alpha=&ClampFloat(Float=$System.InterpolateOverTime.InterpolateOverTime_Float.Moving Average)
    ScaleA=true
    ScaleRGB=true
   )
   /SetVariables(Particles.secondRibbonCOlor:LinearColor=&MultiplyLinearColor(
     Linear Color=<rgba 0.01 0 0 0.35>
     Scale Factor=&MakeFloat4FromVectorAndFloat(
      "Vector (RGB)"=<1 1 1>
      "Float (Alpha)"=&Multiply_Float(A=2, B=&ClampFloat(Float=$System.InterpolateOverTime.InterpolateOverTime_Float.Moving Average))
     )
    )) {isSetParameters=true, module=SetVariables_#5}
   /SetVariables(Particles.ThridRibbonCOlor:LinearColor=&MultiplyLinearColor(
     Linear Color=<rgba 0 0.5776 0.5457 0.5>
     Scale Factor=&MakeFloat4FromVectorAndFloat(
      "Vector (RGB)"=<1 1 1>
      "Float (Alpha)"=&Multiply_Float(A=2, B=&ClampFloat(Float=$System.InterpolateOverTime.InterpolateOverTime_Float.Moving Average))
     )
    )) {isSetParameters=true, module=SetVariables_#7}
  ]
  rendererClasses [&NiagaraRibbonRendererProperties]
  renderers [
   {rendererClass=&NiagaraRibbonRendererProperties}
   {rendererClass=&NiagaraRibbonRendererProperties, rendererIndex=1}
   {rendererClass=&NiagaraRibbonRendererProperties, rendererIndex=2}
   {rendererClass=&NiagaraRibbonRendererProperties, rendererIndex=3}
   {rendererClass=&NiagaraRibbonRendererProperties, rendererIndex=4}
  ]
 }
 FN_AnimTrail_Right {
  emitterSpawnScript [&FN_RibbonIDOffsetUtility()]
  emitterUpdateScript [
   &EmitterState()
   &FN_SpawnAnimTrail(Spawn Mode:ENiagaraSkeletalSpawning=::NewEnumerator0)
  ]
  particleSpawnScript [
   &AddVelocity(Velocity=&Lerp_Vector(A=<0 0 0>, B=<300 -800 0>, Alpha=&FN_FrontEndFloatModifier(Gameplay=0, Front End=1))) {module=AddVelocity001}
   &FN_InitializeAnimTrailBoneIndecies(First Bone or Socket Index=5, Second Bone or Socket Index=2) {module=FN_InitializeAnimTrailBoneIndecies001}
   &FN_InitializeAnimTrail(Color=<rgba 0.9047 0.8148 0.7835 0.35>, Lifetime=0.15)
   /SetVariables(Particles.RibbonTwist2:NiagaraFloat=&AngleConversion(Angle:NiagaraFloat=180, Angle Input:ENiagara_AngleInput=::NewEnumerator0, Angle Output:ENiagara_AngleInput=::NewEnumerator2)) {isSetParameters=true, module=SetVariables_#3}
  ]
  particleUpdateScript [
   &SolveForcesAndVelocity(Clamp Velocity=false, Limit Acceleration=false)
   &FN_UpdateAnimTrail(Locked Particle RibbonWidth=1, SkeletalMesh=$User.SkeletalMesh) {module=FN_UpdateAnimTrail001}
   &ScaleColor(
    Scale Alpha=&ClampFloat(Float=$System.InterpolateOverTime.InterpolateOverTime_Float.Moving Average)
    ScaleA=true
    ScaleRGB=true
   )
   /SetVariables(Particles.secondRibbonCOlor:LinearColor=&MultiplyLinearColor(
     Linear Color=<rgba 0.01 0 0 0.35>
     Scale Factor=&MakeFloat4FromVectorAndFloat(
      "Vector (RGB)"=<1 1 1>
      "Float (Alpha)"=&Multiply_Float(A=2, B=&ClampFloat(Float=$System.InterpolateOverTime.InterpolateOverTime_Float.Moving Average))
     )
    )) {isSetParameters=true, module=SetVariables_#5}
   /SetVariables(Particles.ThridRibbonCOlor:LinearColor=&MultiplyLinearColor(
     Linear Color=<rgba 0 0.5776 0.5457 0.5>
     Scale Factor=&MakeFloat4FromVectorAndFloat(
      "Vector (RGB)"=<1 1 1>
      "Float (Alpha)"=&Multiply_Float(A=3, B=&ClampFloat(Float=$System.InterpolateOverTime.InterpolateOverTime_Float.Moving Average))
     )
    )) {isSetParameters=true, module=SetVariables_#7}
  ]
  rendererClasses [&NiagaraRibbonRendererProperties]
  renderers [
   {rendererClass=&NiagaraRibbonRendererProperties}
   {rendererClass=&NiagaraRibbonRendererProperties, rendererIndex=1}
   {rendererClass=&NiagaraRibbonRendererProperties, rendererIndex=2}
   {rendererClass=&NiagaraRibbonRendererProperties, rendererIndex=3}
   {rendererClass=&NiagaraRibbonRendererProperties, rendererIndex=4}
  ]
 }
 FN_AnimTrail_right_inner {
  emitterSpawnScript [&FN_RibbonIDOffsetUtility()]
  emitterUpdateScript [
   &EmitterState()
   &FN_SpawnAnimTrail(Spawn Mode:ENiagaraSkeletalSpawning=::NewEnumerator0)
  ]
  particleSpawnScript [
   &AddVelocity(Velocity=&Lerp_Vector(A=<0 0 0>, B=<300 -800 0>, Alpha=&FN_FrontEndFloatModifier(Gameplay=0, Front End=1))) {module=AddVelocity001}
   &FN_InitializeAnimTrailBoneIndecies(First Bone or Socket Index=5, Second Bone or Socket Index=3)
   &FN_InitializeAnimTrail(Color=<rgba 0.9047 0.8148 0.7835 0.35>, Lifetime=0.15)
   /SetVariables(Particles.RibbonTwist2:NiagaraFloat=&AngleConversion(Angle:NiagaraFloat=180, Angle Input:ENiagara_AngleInput=::NewEnumerator0, Angle Output:ENiagara_AngleInput=::NewEnumerator2)) {isSetParameters=true, module=SetVariables_#3}
   /SetVariables(Particles.RibbonTwist:NiagaraFloat=&AngleConversion(Angle:NiagaraFloat=180, Angle Input:ENiagara_AngleInput=::NewEnumerator0, Angle Output:ENiagara_AngleInput=::NewEnumerator2)) {isSetParameters=true, module=SetVariables_#4}
  ]
  particleUpdateScript [
   &SolveForcesAndVelocity(Clamp Velocity=false, Limit Acceleration=false)
   &FN_UpdateAnimTrail(Locked Particle RibbonWidth=1, SkeletalMesh=$User.SkeletalMesh) {module=FN_UpdateAnimTrail001}
   &ScaleColor(
    Scale Alpha=&ClampFloat(Float=$System.InterpolateOverTime.InterpolateOverTime_Float.Moving Average)
    ScaleA=true
    ScaleRGB=true
   )
   /SetVariables(Particles.secondRibbonCOlor:LinearColor=&MultiplyLinearColor(
     Linear Color=<rgba 0.01 0 0 0.35>
     Scale Factor=&MakeFloat4FromVectorAndFloat(
      "Vector (RGB)"=<1 1 1>
      "Float (Alpha)"=&Multiply_Float(A=2, B=&ClampFloat(Float=$System.InterpolateOverTime.InterpolateOverTime_Float.Moving Average))
     )
    )) {isSetParameters=true, module=SetVariables_#5}
   /SetVariables(Particles.ThridRibbonCOlor:LinearColor=&MultiplyLinearColor(
     Linear Color=<rgba 0 0.5776 0.5457 0.5>
     Scale Factor=&MakeFloat4FromVectorAndFloat(
      "Vector (RGB)"=<1 1 1>
      "Float (Alpha)"=&Multiply_Float(A=3, B=&ClampFloat(Float=$System.InterpolateOverTime.InterpolateOverTime_Float.Moving Average))
     )
    )) {isSetParameters=true, module=SetVariables_#6}
  ]
  rendererClasses [&NiagaraRibbonRendererProperties]
  renderers [
   {rendererClass=&NiagaraRibbonRendererProperties}
   {rendererClass=&NiagaraRibbonRendererProperties, rendererIndex=1}
   {rendererClass=&NiagaraRibbonRendererProperties, rendererIndex=2}
   {rendererClass=&NiagaraRibbonRendererProperties, rendererIndex=3}
   {rendererClass=&NiagaraRibbonRendererProperties, rendererIndex=4}
  ]
 }
 Side_Tips {
  emitterSpawnScript [&FN_RibbonIDOffsetUtility()]
  emitterUpdateScript [
   &EmitterState()
   &FN_SpawnAnimTrail(Number of Anim Trails=2)
  ]
  particleSpawnScript [
   &AddVelocity(Velocity=&Lerp_Vector(A=<0 0 0>, B=<300 -800 0>, Alpha=&FN_FrontEndFloatModifier(Gameplay=0, Front End=1))) {module=AddVelocity001}
   &FN_InitializeAnimTrailBoneIndecies(AnimTrail Selection=1, First Bone or Socket Index=1, Second Bone or Socket Index=2) {module=FN_InitializeAnimTrailBoneIndecies001}
   &FN_InitializeAnimTrailBoneIndecies(First Bone or Socket Index=3, Second Bone or Socket Index=0)
   &FN_InitializeAnimTrail(Color=<rgba 0.9047 0.8148 0.7835 2>, Lifetime=0.15)
   /SetVariables(Particles.RibbonTwist2:NiagaraFloat=&AngleConversion(Angle:NiagaraFloat=180, Angle Input:ENiagara_AngleInput=::NewEnumerator0, Angle Output:ENiagara_AngleInput=::NewEnumerator2)) {isSetParameters=true, module=SetVariables_#3}
   /SetVariables(Particles.RibbonWidth=5) {isSetParameters=true}
  ]
  particleUpdateScript [
   &SolveForcesAndVelocity(Clamp Velocity=false, Limit Acceleration=false)
   &FN_UpdateAnimTrail(Locked Particle RibbonWidth=1, SkeletalMesh=$User.SkeletalMesh) {module=FN_UpdateAnimTrail001}
   &ScaleColor(
    Scale Alpha=&ClampFloat(Float=$System.InterpolateOverTime.InterpolateOverTime_Float.Moving Average)
    ScaleA=true
    ScaleRGB=true
   )
   /SetVariables(Particles.secondRibbonCOlor:LinearColor=&MultiplyLinearColor(
     Linear Color=<rgba 0 0.5776 0.5457 0.5>
     Scale Factor=&MakeFloat4FromVectorAndFloat(
      "Vector (RGB)"=<1 1 1>
      "Float (Alpha)"=&Multiply_Float(A=2, B=&ClampFloat(Float=$System.InterpolateOverTime.InterpolateOverTime_Float.Moving Average))
     )
    )) {isSetParameters=true, module=SetVariables_#5}
  ]
  rendererClasses [&NiagaraRibbonRendererProperties]
  renderers [
   {rendererClass=&NiagaraRibbonRendererProperties}
   {rendererClass=&NiagaraRibbonRendererProperties, rendererIndex=1}
  ]
 }
}
summary {
 systemName=NS_LazarusLens_Glider_AnimTrail
 emitters [
  {emitterName=FN_AnimTrail_Left, rendererClasses [&NiagaraRibbonRendererProperties]}
  {emitterName=FN_AnimTrail_Left_Inner, rendererClasses [&NiagaraRibbonRendererProperties]}
  {emitterName=Side_Tips, rendererClasses [&NiagaraRibbonRendererProperties]}
  {emitterName=FN_AnimTrail_right_inner, rendererClasses [&NiagaraRibbonRendererProperties]}
  {emitterName=FN_AnimTrail_Right, rendererClasses [&NiagaraRibbonRendererProperties]}
 ]
 userVariables [
  User.bIsFrontEnd {type=NiagaraBool}
  User.bIsFrontEndPreview {type=NiagaraBool}
  User.bIsFullyDeployed {type=NiagaraBool}
  User.ForwardDot {type=NiagaraFloat, defaultValue=0, description=""}
  User.RibbonIDOffset {type=NiagaraInt32}
  User.SkeletalMesh {
   type=NiagaraDataInterfaceSkeletalMesh
   defaultValue {
    struct=NiagaraExt_VariableValue_DataInterface
    value {dataInterface=&self:NiagaraDataInterfaceSkeletalMesh_0}
   }
  }
 ]
}
